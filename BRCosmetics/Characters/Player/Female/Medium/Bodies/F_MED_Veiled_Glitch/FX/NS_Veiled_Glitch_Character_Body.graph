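# Niagara body cosmetic effect for the Veiled Glitch character, attaching glitch particles to the user skeletal mesh.
# Mesh emitters spawn distance-tinted bits around the body and swirl cubes at two bones, alongside two looping bone-pinned sprites.
# GPU digital bits pulse through an expanding spherical mask, with three offset sprite renderers faking chromatic aberration.
# User.bisFootHidden suppresses spawning, alpha or particles within a kill volume; front-end modifiers adjust spawn boxes.
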NiagaraSystem /BRCosmetics/Characters/Player/Female/Medium/Bodies/F_MED_Veiled_Glitch/FX/NS_Veiled_Glitch_Character_Body
dependencies {
 usedDataInterfaces [
  &NiagaraDataInterfaceCurve
  &NiagaraDataInterfaceVector2DCurve
  &NiagaraDataInterfaceColorCurve
  &NiagaraDataInterfaceArrayFloat3
 ]
 usedDynamicInputs [
  &MakeVector4_FromLinearColor
  &FNTimeOfDayScalar
  &FN_FrontEndFloatModifier
  &NormalizeVector
 ]
 usedModules [
  &SystemState
  &self:SystemSpawnScript.SystemScriptSource.SystemScriptGraph.NiagaraNodeAssignment_0.SetVariables_#1
  &self:SystemSpawnScript.SystemScriptSource.SystemScriptGraph.NiagaraNodeAssignment_1.SetVariables_#2
  &self:SystemSpawnScript.SystemScriptSource.SystemScriptGraph.NiagaraNodeAssignment_2.SetVariables_#3
  &InterpolateOverTime
  &self:SystemSpawnScript.SystemScriptSource.SystemScriptGraph.NiagaraNodeAssignment_3.SetVariables_#4
 ]
}
emitters {
 Cube_Glitch_Foot {
  emitterUpdateScript [
   &EmitterState()
   &SpawnRate(
    Spawn Probability=0.5
    SpawnRate=&MakeCustomFloatFromBool(Bool=$User.bisFootHidden, True Float=0, False Float=1)
    Use Spawn Probability=true
   )
   &SpawnRate(
    Spawn Probability=0.5
    SpawnGroup=1
    SpawnRate=&MakeCustomFloatFromBool(Bool=$User.bisFootHidden, True Float=0, False Float=1)
    Use Spawn Probability=true
   ) {module=SpawnRate001}
  ]
  particleSpawnScript [
   /SetVariables(Particles.SpawnGroup=$Emitter.SpawnGroup) {isSetParameters=true}
   &InitializeParticle(
    Lifetime Max=1
    Lifetime Min=0.8
    Lifetime Mode=::NewEnumerator1
    Mesh Scale Mode=::NewEnumerator1
    Mesh Uniform Scale Max=1.5
    UsePositionOffset=true
    Write Material Random=true
   ) {module=InitializeParticle001}
   &ShapeLocation(
    Apply To Particle Position=::NewEnumerator1
    Box Size=&FN_FrontEndVector3Modifier(Gameplay=<20 10 20>, Front End=<20 10 20>)
    MaskSpawnGroup=false
    Offset Coordinate Space=::Local
    Offset Mode=::NewEnumerator0
    Override Seed=false
    Shape Origin=&VectorFromFloat(Value=0)
    Shape Primitive=::NewEnumerator2
    Surface Only=false
    UseOwnerScale=true
   )
   &DynamicMaterialParameters(
    Index 0 Param 4=&RandomRangeFloat(Minimum=-1, Maximum=0.1)
    Index 1 Param 1=&RandomRangeFloat(Minimum=-100, Maximum=100)
    Module.Param0WriteEnabled=true
    Module.Param0WriteEnabled1=true
    Module.Param1WriteEnabled=true
    Module.Param1WriteEnabled1=true
    Module.Param2WriteEnabled=true
    Module.Param2WriteEnabled1=true
    Module.Param3WriteEnabled=true
    Module.Param3WriteEnabled1=true
    Write Parameter Index 1=true
   ) {module=DynamicMaterialParameters001}
  ]
  particleUpdateScript [
   &ParticleState()
   &SkeletalMeshLocation(
    Bone / Socket Index=$Particles.SpawnGroup
    Bone Sampling Mode=::NewEnumerator4
    MaskSpawnGroup=false
    Orientation Sampling=::NewEnumerator1
    Position Sampling=::NewEnumerator1
    Sampled Position Offset=&AddVectorToPosition(
     Position=$Particles.ShapeLocation.ShapePosition
     Vector=&ConvertVectorToPosition(Input Position=<0 0 0>, Conversion Operation:ENiagara_LWCConvertVecToPos=::NewEnumerator0)
    )
    Skeletal Mesh=$User.SkeletalMesh
   ) {module=SkeletalMeshLocation001}
   /SetVariables(Particles.Position=&Add_Vector(
     A=&ConvertPositionToVector(Input Position=$Particles.Position, Convert To Absolute Large World Vector=false)
     B=&Multiply_VectorByFloat(
      Vector=&Add_Vector(
       A=&Lerp_Vector(A=<0 0 1>, B=$System.DiveVelocityScale, Alpha=&Power(a=$Particles.NormalizedAge, b=0.1))
       B=&RotateVectorByQuaternion(
        Vector To Rotate=&SelectVectorFromArray(Direct Array Index=$Particles.SpawnGroup, Vector Selection Array=DataInterface() {propertyValues {FloatData [<0 -0.25 0>, <0 0.25 0>]}})
        Quaternion=$Particles.SkeletalMeshLocation001.SampledRotation
       )
      )
      Float=&FloatFromCurve(
       FloatCurve=DataInterface() {propertyValues {Curve {keys [
           {arriveTangent=2.7561, interpMode=RCIM_Cubic, leaveTangent=2.7561, tangentMode=RCTM_User, time=0, value=-0.1}
           {arriveTangent=-0.9703, interpMode=RCIM_Cubic, leaveTangent=-0.9703, tangentMode=RCTM_User, time=0.1239, value=0.425}
           {arriveTangent=0.031, interpMode=RCIM_Cubic, leaveTangent=0.031, tangentMode=RCTM_User, time=1, value=0}
          ]}}}
       Scale Curve=&Multiply_Float(
        A=&RandomRangeFloat(Minimum=2.5, Maximum=5)
        B=&RandomRangeFloat(Minimum=0.5, Maximum=1)
       )
      )
     )
    )) {isSetParameters=true}
   &RotateAroundPoint(
    Derive the Rotation Axis=false
    Radius:NiagaraFloat=&Multiply_Float(
     A=&Lerp_Float(
      A=&RandomRangeFloat(Minimum=1, Maximum=1.5)
      B=&RandomRangeFloat(Minimum=3, Maximum=6)
      Alpha=&FloatFromCurve(
       FloatCurve=DataInterface() {propertyValues {Curve {keys [
           {arriveTangent=3.8353, interpMode=RCIM_Cubic, leaveTangent=3.8353, tangentMode=RCTM_User, time=0, value=0.5}
           {arriveTangent=-0.5, interpMode=RCIM_Cubic, leaveTangent=-0.5, time=0.3, value=0.9847}
           {interpMode=RCIM_Cubic, leaveTangent=-1.7958, time=1, value=0}
          ]}}}
       Scale Curve=0.1
      )
     )
     B=&MaskFloatBySpawnGroup(Masked Float=1.15, Passthrough Float=0.7, Spawn Group Mask=$Particles.SpawnGroup)
    )
    Rotation Axis:Vector3f=$System.DiveVelocityScale
    Rotation Phase:NiagaraFloat=&RandomRangeFloat(Minimum=0, Maximum=1)
    Rotation Rate:NiagaraFloat=&MakeCustomFloatFromBool(Bool=&RandomBool(), True Float=0.6, False Float=-0.6)
    Set Particle Position=false
   )
   /SetVariables(Particles.Rotational_Position:NiagaraPosition=&ConvertVectorToPosition(
     Input Position=&Add_Vector(
      A=&ConvertPositionToVector(Input Position=$Particles.Position, Convert To Absolute Large World Vector=false)
      B=Linked() {linkedVariable {name=Output.RotateAroundPoint.OutputPosition, type=NiagaraPosition}}
     )
     Vector Is In Absolute Large World Space:NiagaraBool=true
    )) {isSetParameters=true}
   /SetVariables(Particles.ZDistance:NiagaraFloat=&self:ScratchDynamicInput(Distance:NiagaraFloat=150, SK_Mesh_Pos:NiagaraPosition=$Particles.SkeletalMeshLocation001.SampledPosition)) {isSetParameters=true}
   &ScaleMeshSize(Scale Factor=&Multiply_VectorByFloat(
     Vector=&SelectVectorFromArray(
      Direct Array Index=$Particles.SpawnGroup
      Vector Selection Array=DataInterface() {propertyValues {FloatData [<1.5 1.5 1.5>, <1.5 1.5 1.5>, <1 1 1>, <1 1 1>, <0.5 0.5 0.5>, <0.5 0.5 0.5>]}}
     )
     Float=&Multiply_Float(
      A=0.075
      B=&FloatFromCurve(
       FloatCurve=DataInterface() {propertyValues {Curve {keys [
           {arriveTangent=31.886, interpMode=RCIM_Cubic, leaveTangent=31.886, tangentMode=RCTM_User, time=0, value=0}
           {arriveTangent=-1.56, interpMode=RCIM_Cubic, leaveTangent=-1.56, tangentMode=RCTM_User, time=0.2386, value=2.6353}
           {arriveTangent=0.0087, interpMode=RCIM_Cubic, leaveTangent=0.0087, time=0.8412, value=2.2823}
           {arriveTangent=-14.3763, interpMode=RCIM_Cubic, leaveTangent=-14.3763, time=0.9275, value=2.6413}
           {arriveTangent=-0.0405, interpMode=RCIM_Cubic, leaveTangent=-0.0405, tangentMode=RCTM_User, time=1, value=0}
          ]}}}
       Scale Curve=&Multiply_Float(
        A=0.8
        B=&ClampFloat(
         Float=&FN_LOD_Distance_Remap(Start Distance=200, End Distance=400, Minimum Output=0.5, Maximum Output=1)
         Min=0.5
        )
       )
      )
     )
    )) {module=ScaleMeshSize001}
   &ScaleColor(
    Scale Alpha=&MakeCustomFloatFromBool(Bool=$User.bisFootHidden, True Float=0, False Float=1)
    Scale RGB=&MakeVectorFromLinearColorRGB(Color=&LerpLinearColors(StartColor=<rgba 0.9219 0 0.2972 1>, EndColor=<rgba 0 0.803 1 1>, LerpFactor=$Particles.ZDistance))
    ScaleA=true
    ScaleRGB=true
   )
   &CameraOffset(Camera Offset Amount=15)
   &DynamicMaterialParameters(
    Index 0 Param 2=&RandomRangeFloat(Minimum=-1, Maximum=1)
    Index 0 Param 3=&RandomRangeFloat(
     Minimum=&FloatFromCurve(FloatCurve=DataInterface() {propertyValues {Curve {keys [
          <at 0 2 RCIM_Cubic>
          {arriveTangent=-6.2285, interpMode=RCIM_Cubic, leaveTangent=-6.2285, tangentMode=RCTM_User, time=0.7, value=0.875}
          {arriveTangent=-0.8021, interpMode=RCIM_Cubic, leaveTangent=-0.8021, tangentMode=RCTM_User, time=1, value=0.5}
         ]}}})
     Maximum=&FloatFromCurve(FloatCurve=DataInterface() {propertyValues {Curve {keys [{tangentMode=RCTM_User, time=0, value=2}, {tangentMode=RCTM_User, time=1, value=2}]}}})
    )
    Index 1 Param 4=&FloatFromCurve(FloatCurve=DataInterface() {propertyValues {Curve {keys [
         {arriveTangent=4.1785, interpMode=RCIM_Cubic, leaveTangent=4.1785, tangentMode=RCTM_User, time=0, value=0}
         {arriveTangent=0.9995, interpMode=RCIM_Cubic, leaveTangent=0.9995, tangentMode=RCTM_User, time=0.2223, value=0.6825}
         {arriveTangent=-0.0405, interpMode=RCIM_Cubic, leaveTangent=-0.0405, tangentMode=RCTM_User, time=1, value=1}
        ]}}})
    Module.Param0WriteEnabled=false
    Module.Param0WriteEnabled1=false
    Module.Param1WriteEnabled=true
    Module.Param1WriteEnabled1=false
    Module.Param2WriteEnabled=true
    Module.Param2WriteEnabled1=false
    Module.Param3WriteEnabled=true
    Module.Param3WriteEnabled1=true
    Write Parameter Index 1=true
   ) {module=DynamicMaterialParameters002}
  ]
  rendererClasses [&NiagaraMeshRendererProperties]
  renderers [{rendererClass=&NiagaraMeshRendererProperties, rendererIndex=0}]
 }
 Digital_Bits {
  simTarget=GPUComputeSim
  emitterUpdateScript [
   &EmitterState(Inactive Response=::NewEnumerator1, Life Cycle Mode=::NewEnumerator1, Loop Delay=0.5, Loop Duration=0.5, UseLoopDelay=true)
   &SpawnRate(
    SpawnRate=&FloatFromCurve(
     FloatCurve=DataInterface() {propertyValues {Curve {keys [<at 0 1 RCIM_Cubic>, {arriveTangent=-2.7657, interpMode=RCIM_Cubic, leaveTangent=-2.7657, tangentMode=RCTM_User, time=1, value=0}]}}}
     CurveIndex=$Emitter.NormalizedLoopAge
     Scale Curve=&RandomRangeFloat(Minimum=100, Maximum=750)
    )
    Use Spawn Probability=false
   )
  ]
  particleSpawnScript [
   /SetVariables(Particles.ChromaOffset=1) {isSetParameters=true}
   /Niagara/Modules/Spawn/Initialization/InitializeParticle(
    Color=<rgba 1 0 0.7804 0.5>
    Lifetime=&UniformRangedFloat(Minimum=0.05, Maximum=0.2)
    Module.Write Color=true
    Module.Write Lifetime=true
    Module.Write Mass=true
    Module.Write Material Random=false
    Module.Write Position=false
    Module.Write Scale=false
    Module.Write SpriteRotation=false
    Module.Write SpriteSize=true
    Position=&SimulationPosition()
    Sprite Size=&RandomRangeVector2D(Minimum=<0.5 0.1>, Maximum=<1 0.5>)
   )
   &SkeletalMeshLocation(
    Inherit Velocity=false
    MaskSpawnGroup=false
    Mesh Sampling Type=::NewEnumerator7
    Module.Override Randomness=false
    OffsetAlongNormal=false
    Skeletal Mesh=$User.SkeletalMesh
   )
   /SetVariables(
    Particles.NormalVelocity:Vector3f=$Output.SkeletalMeshLocation.SampledNormal
    Particles.TextureSampleA:LinearColor=$Output.SampleTexture.SampledColor
    Particles.TextureSampleB:LinearColor=$Output.SampleTexture001.SampledColor
   ) {isSetParameters=true}
   &CameraOffset(Camera Offset Amount=-10)
   &AddVelocity(Rotation Coordinate Space=::World, Velocity=<0 0 -20>) {module=AddVelocity001}
  ]
  particleUpdateScript [
   &ParticleState()
   &self:SphereMaskEdge(
    Cull Particles outside of max distance=true
    Max Distance:NiagaraFloat=&FloatFromCurve(
     FloatCurve=DataInterface() {propertyValues {Curve {keys [
         {arriveTangent=1.1063, interpMode=RCIM_Cubic, leaveTangent=1.1063, tangentMode=RCTM_User, time=0.001, value=0}
         {arriveTangent=1.1925, interpMode=RCIM_Cubic, leaveTangent=1.1925, tangentMode=RCTM_User, time=0.9, value=1}
        ]}}}
     CurveIndex=$Emitter.NormalizedLoopAge
     Scale Curve=128
    )
    Min Distance:NiagaraFloat=&FloatFromCurve(
     FloatCurve=DataInterface() {propertyValues {Curve {keys [
         {arriveTangent=0.7543, interpMode=RCIM_Cubic, leaveTangent=0.7543, tangentMode=RCTM_User, time=0.004, value=0}
         {arriveTangent=1.004, interpMode=RCIM_Cubic, leaveTangent=1.004, time=0.6, value=0.451}
         {arriveTangent=1.4135, interpMode=RCIM_Cubic, leaveTangent=1.4135, tangentMode=RCTM_User, time=1, value=1}
        ]}}}
     CurveIndex=$Emitter.NormalizedLoopAge
     Scale Curve=128
    )
    SphereCenter:Vector3f=<0 8.5591 128>
   ) {module=SphereMaskEdge}
   &AddVelocity(Velocity=&Multiply_Vector(A=$Particles.NormalVelocity, B=&VectorFromFloat(Value=-1)))
   &SolveForcesAndVelocity(Clamp Velocity=false, Limit Acceleration=false)
   /SetVariables(Particles.ChromaOffset:NiagaraFloat=&FloatFromCurve(
     FloatCurve=DataInterface() {propertyValues {Curve {keys [{interpMode=RCIM_Cubic, tangentMode=RCTM_User, time=0, value=1}, {interpMode=RCIM_Cubic, tangentMode=RCTM_User, time=1, value=0}]}}}
     Scale Curve=0.5
    )) {isSetParameters=true}
   /SetVariables(
    Particles.Color001:LinearColor=<rgba 0.1 0 0 0.5>
    Particles.Position001:NiagaraPosition=&ConvertVectorToPosition(
     Input Position=&Add_Vector(
      A=&ConvertPositionToVector(Input Position=$Particles.Position, Convert To Absolute Large World Vector=false)
      B=&Multiply_VectorByFloat(Vector=<2 0 1>, Float=$Particles.ChromaOffset)
     )
     Vector Is In Absolute Large World Space:NiagaraBool=false
    )
   ) {isSetParameters=true}
   /SetVariables(
    Particles.Color002=<rgba 0 0.0077 0.1 0.5>
    Particles.Position002:NiagaraPosition=&ConvertVectorToPosition(
     Input Position=&Add_Vector(
      A=&ConvertPositionToVector(Input Position=$Particles.Position, Convert To Absolute Large World Vector=false)
      B=&Multiply_VectorByFloat(Vector=<-2 -1 -2>, Float=$Particles.ChromaOffset)
     )
     Vector Is In Absolute Large World Space:NiagaraBool=false
    )
   ) {isSetParameters=true}
   /SetVariables(
    Particles.Color003:LinearColor=<rgba 0.0029 0.1 0 0.5>
    Particles.Position003:NiagaraPosition=&ConvertVectorToPosition(
     Input Position=&Add_Vector(
      A=&ConvertPositionToVector(Input Position=$Particles.Position, Convert To Absolute Large World Vector=false)
      B=&Multiply_VectorByFloat(Vector=<1 2 -2>, Float=$Particles.ChromaOffset)
     )
     Vector Is In Absolute Large World Space:NiagaraBool=false
    )
   ) {isSetParameters=true}
   &ScaleSpriteSize(Scale Factor=&Vector2DFromCurve(
     Vector2Curve=DataInterface() {propertyValues {
       XCurve {keys [
         {interpMode=RCIM_Cubic, tangentMode=RCTM_User, time=0, value=1}
         {arriveTangent=-6.968, interpMode=RCIM_Cubic, leaveTangent=-6.968, tangentMode=RCTM_User, time=0.5, value=2}
         {tangentMode=RCTM_User, time=1, value=0}
        ]}
       YCurve {keys [
         {interpMode=RCIM_Cubic, tangentMode=RCTM_User, time=0, value=0}
         {arriveTangent=13.8249, interpMode=RCIM_Cubic, leaveTangent=13.8249, tangentMode=RCTM_User, time=0.6, value=2}
         {arriveTangent=8.3745, interpMode=RCIM_Cubic, leaveTangent=8.3745, tangentMode=RCTM_User, time=1, value=8}
        ]}
      }}
     ScaleCurve=<2 8>
    ))
   /SetVariables(Particles.SpriteSize001:Vector2f=&Multiply_Vector2D_ByFloat(Float=1, Vector2D=$Particles.SpriteSize)) {isSetParameters=true}
   &KillParticlesInVolume(Kill Shape=::NewEnumerator2, Kill Volume Enabled:NiagaraBool=$User.bisFootHidden, Origin Offset:Vector3f=<0 0 65>)
  ]
  rendererClasses [&NiagaraSpriteRendererProperties]
  renderers [
   {rendererClass=&NiagaraSpriteRendererProperties}
   {rendererClass=&NiagaraSpriteRendererProperties, rendererIndex=1}
   {rendererClass=&NiagaraSpriteRendererProperties, rendererIndex=2}
  ]
 }
 Glitch_Decals {
  emitterUpdateScript [
   &EmitterState(
    Inactive Response=::NewEnumerator1
    Life Cycle Mode=::NewEnumerator1
    Loop Behavior=::NewEnumerator1
    Loop Duration Mode=::NewEnumerator1
    UseLoopDelay=false
   )
   &SpawnBurst_Instantaneous(Spawn Count=2, Use Spawn Probability=false, UseLoopCountLimit=false)
  ]
  particleSpawnScript [
   /SetVariables(Particles.Random=&RandomRangeFloat(Minimum=0, Maximum=1)) {isSetParameters=true}
   &FN_SetSpawnGroup(Number of Spawn Groups=2)
   &InitializeParticle(
    Color=<rgba 1 0 0.9175 0.8>
    Lifetime=&Lerp_Float(A=1, B=0.5, Alpha=$Particles.Random)
    Module.Write Color=true
    Module.Write Lifetime=true
    Module.Write Mass=true
    Module.Write Material Random=true
    Module.Write Position=false
    Module.Write Scale=false
    Module.Write SpriteRotation=false
    Module.Write SpriteSize=true
    Position=&SimulationPosition()
    Sprite Size=<35 35>
   )
   &CameraOffset(Camera Offset Amount=30)
  ]
  particleUpdateScript [
   &ParticleState(
    Kill Particles When Lifetime Has Elapsed=false
    Let Infinitely Lived Particles Die When Emitter Deactivates=true
    Loop Particles Lifetime=true
   )
   &SkeletalMeshLocation(
    Bone / Socket Index=$Particles.SpawnGroup
    Bone Sampling Mode=::NewEnumerator4
    Inherit Velocity=false
    MaskSpawnGroup=false
    Sampled Position Offset=<0 5 25>
    Skeletal Mesh=$User.SkeletalMesh
   )
   &SolveForcesAndVelocity(Clamp Velocity=false, Limit Acceleration=false)
   &DynamicMaterialParameters(
    Index 0 Param 1=&Lerp_Float(A=0.2, B=1.1, Alpha=$Particles.Random)
    Index 0 Param 2=&Lerp_Float(A=1, B=2, Alpha=$Particles.Random)
    Index 0 Param 3=&Lerp_Float(A=0.3, B=1, Alpha=$Particles.Random)
    Index 0 Param 4=-1
    Module.Param0WriteEnabled=true
    Module.Param1WriteEnabled=true
    Module.Param2WriteEnabled=true
    Module.Param3WriteEnabled=true
   )
   &ScaleColor(
    Scale Alpha=&MakeCustomFloatFromBool(Bool=$User.bisFootHidden, True Float=0, False Float=1)
    ScaleA=true
    ScaleRGB=true
   )
   /SetVariables(Particles.VisibilityTag=&MakeCustomIntFromBool(Bool=$User.bisFootHidden, True INT=1, False INT=0)) {isSetParameters=true}
  ]
  rendererClasses [&NiagaraSpriteRendererProperties]
  renderers [{rendererClass=&NiagaraSpriteRendererProperties}]
 }
 Main_Glitch_Bits__WORLD_SPACE_ {
  emitterUpdateScript [
   &EmitterState(Life Cycle Mode=::NewEnumerator1, Loop Delay=0.2, UseLoopDelay=true)
   &SpawnRate(
    Spawn Probability=0.5
    SpawnRate=&MakeCustomFloatFromBool(Bool=$User.bisFootHidden, True Float=0, False Float=1)
    Use Spawn Probability=true
   )
  ]
  particleSpawnScript [
   /Niagara/Modules/Spawn/Initialization/V2/InitializeParticle(
    Lifetime Max=0.75
    Lifetime Min=0.5
    Lifetime Mode=::NewEnumerator1
    Mesh Scale Mode=::NewEnumerator1
    Mesh Uniform Scale Max=1.5
    UsePositionOffset=true
    Write Material Random=true
   )
   &ShapeLocation(
    Apply To Particle Position=::NewEnumerator1
    Box Size=&FN_FrontEndVector3Modifier(Gameplay=<20 20 50>, Front End=<10 10 40>)
    MaskSpawnGroup=false
    Override Seed=false
    Shape Origin=&VectorFromFloat(Value=0)
    Shape Primitive=::NewEnumerator2
    Surface Only=false
    UseOwnerScale=true
   )
   &SkeletalMeshLocation(
    Bone Sampling Mode=::NewEnumerator0
    MaskSpawnGroup=false
    Orientation Sampling=::NewEnumerator1
    OverrideRandomSeed=false
    Sampled Position Offset=&AddVectorToPosition(
     Position=Linked() {linkedVariable {name=Particles.ShapeLocation.ShapePosition, type=Vector3f}}
     Vector=&ConvertVectorToPosition(Input Position=<0 0 0>, Conversion Operation:ENiagara_LWCConvertVecToPos=::NewEnumerator0)
    )
    Skeletal Mesh=$User.SkeletalMesh
    Velocity Sampling=::NewEnumerator0
    VelocitySpeedLimit=false
   )
   &DynamicMaterialParameters(
    Index 0 Param 4=&RandomRangeFloat(Minimum=-1, Maximum=0.1)
    Index 1 Param 1=&RandomRangeFloat(Minimum=-100, Maximum=100)
    Module.Param0WriteEnabled=true
    Module.Param0WriteEnabled1=true
    Module.Param1WriteEnabled=true
    Module.Param1WriteEnabled1=true
    Module.Param2WriteEnabled=true
    Module.Param2WriteEnabled1=true
    Module.Param3WriteEnabled=true
    Module.Param3WriteEnabled1=true
    Write Parameter Index 1=true
   )
   &AddVelocity(Rotation Coordinate Space=::Simulation, Velocity=<0 0 1>, Velocity Speed Scale=500)
   &AddVelocity(
    Module.Use Falloff Distance=false
    Origin Offset=&RandomRangeVector(Minimum=<0 25 50>, Maximum=<0 50 100>)
    Override Seed=false
    Velocity Mode=::NewEnumerator1
    Velocity Speed=&RandomRangeFloat(Minimum=25, Maximum=50)
   ) {module=AddVelocity002}
  ]
  particleUpdateScript [
   /SetVariables(Particles.ZDistance:NiagaraFloat=&self:ScratchDynamicInput(Distance:NiagaraFloat=150, SK_Mesh_Pos:NiagaraPosition=$Particles.SkeletalMeshLocation.SampledPosition)) {isSetParameters=true}
   &ParticleState()
   &ScaleMeshSize(Scale Factor=&Multiply_VectorByFloat(
     Vector=<1 1 1>
     Float=&Multiply_Float(
      A=0.13
      B=&FloatFromCurve(
       FloatCurve=DataInterface() {propertyValues {Curve {keys [
           {arriveTangent=31.886, interpMode=RCIM_Cubic, leaveTangent=31.886, tangentMode=RCTM_User, time=0, value=0}
           {arriveTangent=-1.56, interpMode=RCIM_Cubic, leaveTangent=-1.56, tangentMode=RCTM_User, time=0.2386, value=2.6353}
           {arriveTangent=0.0087, interpMode=RCIM_Cubic, leaveTangent=0.0087, time=0.8412, value=2.2823}
           {arriveTangent=-14.3763, interpMode=RCIM_Cubic, leaveTangent=-14.3763, time=0.9275, value=2.6413}
           {arriveTangent=-0.0405, interpMode=RCIM_Cubic, leaveTangent=-0.0405, tangentMode=RCTM_User, time=1, value=0}
          ]}}}
       Scale Curve=&Multiply_Float(
        A=1
        B=&ClampFloat(
         Float=&FN_LOD_Distance_Remap(Start Distance=200, End Distance=400, Minimum Output=0.5, Maximum Output=1)
         Min=0.5
        )
       )
      )
     )
    ))
   &Drag(Drag=25, Use Linear Drag=true, Use Rotational Drag=true)
   &ScaleVelocity(Velocity Scale=&Multiply_VectorByFloat(
     Vector=<1 1 0.9>
     Float=&FloatFromCurve(FloatCurve=DataInterface() {propertyValues {Curve {keys [
          {arriveTangent=1.3524, interpMode=RCIM_Cubic, leaveTangent=1.3524, tangentMode=RCTM_User, time=0, value=0.5}
          {arriveTangent=1.5579, interpMode=RCIM_Cubic, leaveTangent=1.5579, time=0.3395, value=0.7523}
          {arriveTangent=3.4032, interpMode=RCIM_Cubic, leaveTangent=3.4032, time=0.6419, value=1.5}
          {interpMode=RCIM_Cubic, tangentMode=RCTM_User, time=1, value=3}
         ]}}})
    )) {enabled=false}
   &SolveForcesAndVelocity(Clamp Velocity=false, Limit Acceleration=false)
   &ScaleColor(
    Scale Alpha=&MakeCustomFloatFromBool(Bool=$User.bisFootHidden, True Float=0, False Float=1)
    Scale RGB=&MakeVectorFromLinearColorRGB(Color=&LerpLinearColors(StartColor=<rgba 1 0 0.3099 1>, EndColor=<rgba 0.2413 0 1 1>, LerpFactor=$Particles.ZDistance))
    ScaleA=true
    ScaleRGB=true
   )
   &DynamicMaterialParameters(
    Index 0 Param 2=&RandomRangeFloat(Minimum=-1, Maximum=1)
    Index 0 Param 3=&RandomRangeFloat(
     Minimum=&FloatFromCurve(FloatCurve=DataInterface() {propertyValues {Curve {keys [
          <at 0 2 RCIM_Cubic>
          {arriveTangent=-6.2285, interpMode=RCIM_Cubic, leaveTangent=-6.2285, tangentMode=RCTM_User, time=0.7, value=0.875}
          {arriveTangent=-0.8021, interpMode=RCIM_Cubic, leaveTangent=-0.8021, tangentMode=RCTM_User, time=1, value=0.5}
         ]}}})
     Maximum=&FloatFromCurve(FloatCurve=DataInterface() {propertyValues {Curve {keys [{tangentMode=RCTM_User, time=0, value=2}, {tangentMode=RCTM_User, time=1, value=2}]}}})
    )
    Index 1 Param 4=&FloatFromCurve(FloatCurve=DataInterface() {propertyValues {Curve {keys [
         {arriveTangent=4.1785, interpMode=RCIM_Cubic, leaveTangent=4.1785, tangentMode=RCTM_User, time=0, value=0}
         {arriveTangent=0.9995, interpMode=RCIM_Cubic, leaveTangent=0.9995, tangentMode=RCTM_User, time=0.2223, value=0.6825}
         {arriveTangent=-0.0405, interpMode=RCIM_Cubic, leaveTangent=-0.0405, tangentMode=RCTM_User, time=1, value=1}
        ]}}})
    Module.Param0WriteEnabled=false
    Module.Param0WriteEnabled1=false
    Module.Param1WriteEnabled=true
    Module.Param1WriteEnabled1=false
    Module.Param2WriteEnabled=true
    Module.Param2WriteEnabled1=false
    Module.Param3WriteEnabled=true
    Module.Param3WriteEnabled1=true
    Write Parameter Index 1=true
   ) {module=DynamicMaterialParameters001}
  ]
  rendererClasses [&NiagaraMeshRendererProperties]
  renderers [{rendererClass=&NiagaraMeshRendererProperties, rendererIndex=0}]
 }
}
summary {
 systemName=NS_Veiled_Glitch_Character_Body
 emitters [
  {emitterName=Digital_Bits, simTarget=GPUComputeSim, rendererClasses [&NiagaraSpriteRendererProperties]}
  {emitterName=Glitch_Decals, rendererClasses [&NiagaraSpriteRendererProperties]}
  {emitterName=Main_Glitch_Bits__WORLD_SPACE_, rendererClasses [&NiagaraMeshRendererProperties]}
  {emitterName=Cube_Glitch_Foot, rendererClasses [&NiagaraMeshRendererProperties]}
 ]
 userVariables [
  User.bisFootHidden {type=NiagaraBool}
  User.bIsFrontEnd {type=NiagaraBool}
  User.bIsFrontEndPreview {type=NiagaraBool}
  User.SkeletalMesh {
   type=NiagaraDataInterfaceSkeletalMesh
   defaultValue {
    struct=NiagaraExt_VariableValue_DataInterface
    value {dataInterface=&self:NiagaraDataInterfaceSkeletalMesh_0}
   }
  }
 ]
}
