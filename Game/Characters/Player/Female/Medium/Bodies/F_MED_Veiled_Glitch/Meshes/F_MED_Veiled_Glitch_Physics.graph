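# Physics asset for the F_MED_Veiled_Glitch female player body, covering the left and right hip ribbons.
# Each ribbon is a chain of four capsules running from a small base sphere through parent, 01, 02 and 03 bones.
# Narrow swing and very tight twist limits keep the chains stiff, with one swing axis tighter near the root.
# A pelvis sphere is the only other body, and no constraint links it to the ribbons.
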PhysicsAsset /Game/Characters/Player/Female/Medium/Bodies/F_MED_Veiled_Glitch/Meshes/F_MED_Veiled_Glitch_Physics
bodies {
 dyn_hip_ribbon_01_l [{shapeType=Capsule, length=9.2782, radius=2.7515, shapeName=dyn_hip_ribbon_01_l_capsule, center=<6.1439 0 0>, rotation=<rot 90 0 0>}]
 dyn_hip_ribbon_01_r [{shapeType=Capsule, length=8.3951, radius=2.7324, shapeName=dyn_hip_ribbon_01_r_capsule, center=<6.1439 0 0>, rotation=<rot 90 0 0>}]
 dyn_hip_ribbon_02_l [{shapeType=Capsule, length=10.4491, radius=2.855, shapeName=dyn_hip_ribbon_02_l_capsule, center=<6.1439 0 0>, rotation=<rot 90 0 0>}]
 dyn_hip_ribbon_02_r [{shapeType=Capsule, length=9.1349, radius=2.8466, shapeName=dyn_hip_ribbon_02_r_capsule, center=<6.1439 0 0>, rotation=<rot 90 0 0>}]
 dyn_hip_ribbon_03_l [{shapeType=Capsule, length=10.6601, radius=3.8114, shapeName=dyn_hip_ribbon_03_l_capsule, center=<6.1439 0 0>, rotation=<rot 90 0 0>}]
 dyn_hip_ribbon_03_r [{shapeType=Capsule, length=10.6567, radius=3.8102, shapeName=dyn_hip_ribbon_03_r_capsule, center=<6.1439 0 0>, rotation=<rot 90 0 0>}]
 dyn_hip_ribbon_base_l [{shapeType=Sphere, radius=0.5, shapeName=dyn_hip_ribbon_base_l_sphere}]
 dyn_hip_ribbon_base_r [{shapeType=Sphere, radius=0.5, shapeName=dyn_hip_ribbon_base_r_sphere}]
 dyn_hip_ribbon_parent_l [{shapeType=Capsule, length=7.3979, radius=2.7056, shapeName=dyn_hip_ribbon_parent_l_capsule, center=<6.1439 0 0>, rotation=<rot 90 0 0>}]
 dyn_hip_ribbon_parent_r [{shapeType=Capsule, length=8.543, radius=2.8329, shapeName=dyn_hip_ribbon_parent_r_capsule, center=<6.1439 0 0>, rotation=<rot 90 0 0>}]
 pelvis [{shapeType=Sphere, radius=5.5326, shapeName=pelvis_sphere}]
}
constraints [
 {bone1Name=dyn_hip_ribbon_03_l, bone2Name=dyn_hip_ribbon_02_l, swing1LimitDegrees=10, swing2LimitDegrees=10, twistLimitDegrees=2}
 {bone1Name=dyn_hip_ribbon_parent_l, bone2Name=dyn_hip_ribbon_base_l, swing1LimitDegrees=10, swing2LimitDegrees=5, twistLimitDegrees=2}
 {bone1Name=dyn_hip_ribbon_02_l, bone2Name=dyn_hip_ribbon_01_l, swing1LimitDegrees=10, swing2LimitDegrees=10, twistLimitDegrees=2}
 {bone1Name=dyn_hip_ribbon_01_l, bone2Name=dyn_hip_ribbon_parent_l, swing1LimitDegrees=10, swing2LimitDegrees=5, twistLimitDegrees=2}
 {bone1Name=dyn_hip_ribbon_parent_r, bone2Name=dyn_hip_ribbon_base_r, swing1LimitDegrees=10, swing2LimitDegrees=5, twistLimitDegrees=2}
 {bone1Name=dyn_hip_ribbon_01_r, bone2Name=dyn_hip_ribbon_parent_r, swing1LimitDegrees=10, swing2LimitDegrees=5, twistLimitDegrees=2}
 {bone1Name=dyn_hip_ribbon_02_r, bone2Name=dyn_hip_ribbon_01_r, swing1LimitDegrees=10, swing2LimitDegrees=10, twistLimitDegrees=2}
 {bone1Name=dyn_hip_ribbon_03_r, bone2Name=dyn_hip_ribbon_02_r, swing1LimitDegrees=10, swing2LimitDegrees=10, twistLimitDegrees=2}
]
