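# Animation Blueprint for the F_MED_Veiled_Glitch female character body, copying the attached parent's pose and curves and adding pelvis-space RigidBody secondary physics.
# Two simulations share one physics asset: the cape-worn branch uses fixed gravity, the other takes a per-frame gravity override from a deprecated state function with custom floating-skydive wind.
# On customization load, a backbling check sets the pack flag and the long-backbling cape flag that selects the branch.
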AnimBlueprint /BRCosmetics/Characters/Player/Female/Medium/Bodies/F_MED_Veiled_Glitch/Meshes/F_MED_Veiled_Glitch_AnimBP
target_skeleton=/Game/Characters/Player/Female/Medium/Bodies/F_MED_Veiled_Glitch/Meshes/F_MED_Veiled_Glitch_Skeleton
graphs [
 AnimGraph {nodes [
   n1 = RigidBody!(
    ComponentPose=Local To Component(LocalPose=Use cached pose 'GravityCache'())
    Alpha
    OverrideWorldGravity:Vector=Get Gravity_Override() -> (Gravity_Override:Vector)
    ExternalForce
   ) -> (Pose) {
    bClampLinearTranslationLimitToRefPose=true
    bOverrideWorldGravity=true
    overridePhysicsAsset=/Game/Characters/Player/Female/Medium/Bodies/F_MED_Veiled_Glitch/Meshes/F_MED_Veiled_Glitch_Physics
    simulationSpace=BaseBoneSpace
    alphaBoolBlend {blendOption=Linear}
    baseBoneRef {boneName=pelvis}
    componentAppliedLinearAccClamp=<6000 6000 6000>
    componentLinearAccScale=<0.1 0.1 0.1>
    componentLinearVelScale=<0.1 0.1 0.1>
    overrideWorldGravity {z=-980}
    simSpaceSettings {
     maxAngularAcceleration=500
     maxAngularVelocity=3
     maxLinearAcceleration=400
     maxLinearVelocity=300
     velocityScaleZ=0
     worldAlpha=1
     externalLinearDragV=<0.1 0.1>
    }
   }
   n2 = RigidBody!(ComponentPose=Local To Component(LocalPose=Use cached pose 'GravityCache'()), Alpha, ExternalForce) -> (Pose) {
    bClampLinearTranslationLimitToRefPose=true
    bOverrideWorldGravity=true
    overridePhysicsAsset=&F_MED_Veiled_Glitch_Physics
    simulationSpace=BaseBoneSpace
    alphaBoolBlend {blendOption=Linear}
    baseBoneRef {boneName=pelvis}
    componentAppliedLinearAccClamp=<6000 6000 6000>
    componentLinearAccScale=<0.15 0.15 0.15>
    componentLinearVelScale=<0.15 0.15 0.15>
    overrideWorldGravity {z=-980}
    simSpaceSettings {
     maxAngularAcceleration=500
     maxAngularVelocity=3
     maxLinearAcceleration=400
     maxLinearVelocity=250
     velocityScaleZ=0
     worldAlpha=1
     externalLinearDragV=<0.1 0.1>
    }
   }
   Output Pose(Result=Blend Poses by bool(
     BlendPose_0=Component To Local(ComponentPose=@n2)
     BlendPose_1=Component To Local(ComponentPose=@n1)
     bActiveValue=Get isWearingCape() -> (isWearingCape:bool)
    ) {blendType=Linear}) {layerGroup=None}
   Save cached pose 'GravityCache'(Pose=Copy Pose From Mesh() {bCopyCurves=true, bUseAttachedParent=true}) {cacheName=GravityCache}
  ]}
 EventGraph {nodes [
   Event BlueprintUpdateAnimation()
   > n1 = Gravity Override State Function DEPRECATED(CharacterPartAnimInstance=Self-Reference(), GameTimeInSeconds=GetGameTimeInSeconds(), GravityOverrideParams=Get Gravity Override Params())
   > Set Gravity_Override(Gravity_Override:Vector=@n1) -> (Output_Get:Vector)
   Event OnCharacterCustomizationLoaded()
   n2 = CheckForBackbling()
   > Set isWearingPack(isWearingPack:bool=@n2) -> (Output_Get:bool)
   > Set isWearingCape(isWearingCape=AND Boolean(A=@n2, B=@n2.bIsBackblingLong))
  ]}
]
variables [
 Gravity Override Params {
  type=GravityOverrideParamsStruct
  default {skydive_FloatingDown {gravityOverride=<0 -200 400>, windAmplitude=<200 -200 200>, windFrequency=<6 10 8>, windOffset=<0 -200 200>}}
 }
 Gravity_Override {type=Vector}
 isWearingCape {type=boolean}
 isWearingPack {type=boolean}
]
